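# Sound cue for the generic front-end toast-appear whoosh in the UI.
# A single wave player plays the UI_PartyInvite_Appear wave from the ItemShop sound folder, reusing that asset rather than a dedicated toast sound.
# It routes through the Fort_UI sound class, with no randomization, modulation or mixing nodes.
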SoundCue /Game/Sounds/Fort_UI_Sounds/FrontEnd/Whooshes/Toasts/UI_Whoosh_Generic_Toast_Appear_Cue
root=WavePlayer_1
nodes [{id=WavePlayer_1, properties {soundWaveAssetPtr=/Game/Sounds/Fort_UI_Sounds/FrontEnd/ItemShop/wav/UI_PartyInvite_Appear}}]
properties {maxDistance=2097152, soundClassObject=&Fort_UI}
