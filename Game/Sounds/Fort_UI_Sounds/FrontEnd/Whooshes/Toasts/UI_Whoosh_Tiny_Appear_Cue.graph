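# UI_Whoosh_Tiny_Appear_Cue is a front-end UI SoundCue, named and filed as a tiny whoosh for toast appearance.
# Its graph has no root or nodes, so it references no audio and produces no sound.
# It keeps the Fort_UI sound class and overrides concurrency to prevent new instances, with a short retrigger time.
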SoundCue /Game/Sounds/Fort_UI_Sounds/FrontEnd/Whooshes/Toasts/UI_Whoosh_Tiny_Appear_Cue
root=null
nodes []
properties {
 bOverrideConcurrency=true
 maxDistance=2097152
 soundClassObject=&Fort_UI
 volumeMultiplier=1
 concurrencyOverrides {resolutionRule=PreventNew, retriggerTime=0.15}
}
