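# Sound cue for a tiny front-end UI whoosh, named and filed as the close sound for toasts.
# It has no nodes and a null root, so no sound wave is wired in and the cue produces no audio as it stands.
# It retains playback settings: the Fort_UI sound class, lowered volume, raised pitch and a concurrency override allowing one instance, stopping the oldest, with a short retrigger time.
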SoundCue /Game/Sounds/Fort_UI_Sounds/FrontEnd/Whooshes/Toasts/UI_Whoosh_Tiny_Close_Cue
root=null
nodes []
properties {
 bOverrideConcurrency=true
 maxDistance=2097152
 pitchMultiplier=2
 soundClassObject=&Fort_UI
 volumeMultiplier=0.25
 concurrencyOverrides {maxCount=1, resolutionRule=StopOldest, retriggerTime=0.2}
}
